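; Base view model Blueprint for the About Item dialog shown in frontend item preview.
; It defines two functions that return the dialog's entries: one for design-time preview and one taking a Fort item view model.
; Both simply return the Entries variable, and the item parameter is unused here.
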(fn GetDesignTimeEntries
  (return (Variables|Entries|GetEntries)))

(fn GetEntriesForItem
  (params (ItemVM :type Fort%20Item%20VM%20Object%20Reference))
  (return (Variables|Entries|GetEntries)))
