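# Body entry widget for the item preview About Item dialog, showing a title and body text with an optional side image.
# InitializeEntry sets both auto-wrapping texts, which collapse when empty, and checks whether the soft texture reference is valid.
# With a valid image, the lazy image shows and loads it without matching size; otherwise it collapses and the text size box widens to use the space.
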WidgetBlueprint /FortniteCoreUI/Frontend/ItemPreview/AboutItemDialog/WBP_AboutItemDialog_BodyEntry
graphs [
 EventGraph {nodes [
   Event PreConstruct()
   Event Construct()
   Event Tick()
  ]}
 InitializeEntry {nodes [
   n1 = InitializeEntry!() -> (then, Title:Text, BodyText:Text, Image:Texture 2D Soft Object Reference) {is=title K2Node_FunctionEntry}
   > Set Text(Text=@n1, self=Get Text_Title() -> (Text_Title:Common Text Object Reference))
   > Set Text(Text=Get BodyText() -> (BodyText:Text), self=Get Text_Body() -> (Text_Body:Common Text Object Reference))
   > Set HasValidImage(HasValidImage:bool=IsValidSoftObjectReference(SoftObjectReference=Get Image() -> (Image:Texture 2D Soft Object Reference))) -> (Output_Get:bool)
   > Set Visibility(
    Visibility=Select(Option 0=Collapsed, Option 1=SelfHitTestInvisible, Index=Get HasValidImage() -> (HasValidImage:bool))
    self=Get LazyImage_SideImage() -> (LazyImage_SideImage:Fort Lazy Image Object Reference)
   )
   > Set WidthOverride!(
    execute
    WidthOverride:float=SelectFloat(A=534, B=784, bPickA=Get HasValidImage())
    self:Size Box Object Reference=Get SizeBox_Text() -> (SizeBox_Text:Size Box Object Reference)
   ) -> (then, Output_Get:float)
   > Branch(Condition=Get HasValidImage())
   > SetBrushFromLazyTexture!(
    execute
    self:Common Lazy Image Object Reference=Get LazyImage_SideImage()
    LazyTexture:"Texture 2D Soft Object Reference (by ref)"=Get Image()
    bMatchSize:bool=false
   ) -> (then)
  ]}
]
variables []
widgets {
 info {inheritedWidgetCount=0, parentClass=&UserWidget, rootWidgetClass=&HorizontalBox, widgetCount=6}
 widgets [
  HorizontalBox_All {widgetClassPath=&HorizontalBox, widget=HorizontalBox_All, properties {visibility=HitTestInvisible}}
  LazyImage_SideImage {
   widgetClassPath=&FortLazyImage
   bIsVariable=true
   parent=HorizontalBox_All
   widget=LazyImage_SideImage
   properties {brush {
     bIsDynamicallyLoaded=false
     drawAs=Image
     imageType=NoImage
     mirroring=NoMirror
     resourceName=None
     resourceObject=null
     tiling=NoTile
     imageSize=<172 172>
     margin=<ltrb 0 0 0 0>
     outlineSettings {
      bUseBrushTransparency=false
      roundingType=HalfHeightRadius
      width=0
      color {colorUseRule=UseColor_Specified, specifiedColor=<rgba 0 0 0 0>}
      cornerRadii=<0 0 0 0>
     }
     tintColor {colorUseRule=UseColor_Specified, specifiedColor=<rgba 1 1 1 1>}
     uVRegion {bIsValid=false, max=<0 0>, min=<0 0>}
    }}
   slot {class=HorizontalBoxSlot, properties {horizontalAlignment=HAlign_Left, verticalAlignment=VAlign_Center, padding=<ltrb 0 0 32 0>}}
  }
  SizeBox_Text {
   widgetClassPath=&SizeBox
   bIsVariable=true
   parent=HorizontalBox_All
   widget=SizeBox_Text
   properties {bOverride_WidthOverride=true, widthOverride=534}
   slot {class=HorizontalBoxSlot}
  }
  VerticalBox_49 {
   widgetClassPath=&VerticalBox
   parent=SizeBox_Text
   widget=VerticalBox_49
   slot {class=SizeBoxSlot, properties {verticalAlignment=VAlign_Center}}
  }
  Text_Title {
   widgetClassPath=&CommonTextBlock
   bIsVariable=true
   parent=VerticalBox_49
   widget=Text_Title
   properties {
    autoWrapText=true
    bAutoCollapseWithEmptyText=true
    bIsScrollingEnabled=false
    mobileFontSizeMultiplier=1.2
    text="[PH] Entry title"
    colorAndOpacity {colorUseRule=UseColor_Specified, specifiedColor=<rgba 0 1 0.9647 1>}
    font {
     bForceMonospaced=false
     bMaterialIsStencil=false
     fontMaterial=null
     fontObject=&HeadingNow
     letterSpacing=0
     monospacedWidth=1
     size=18
     skewAmount=0
     typefaceFontName=75 Medium
     outlineSettings {
      bApplyOutlineToDropShadows=false
      bMiteredCorners=false
      bSeparateFillAlpha=false
      outlineMaterial=null
      outlineSize=0
      outlineColor=<rgba 0 0 0 1>
     }
    }
    shadowOffset=<0 0>
   }
   slot {class=VerticalBoxSlot, properties {padding=<ltrb 0 0 0 20>}}
  }
  Text_Body {
   widgetClassPath=&CommonTextBlock
   bIsVariable=true
   parent=VerticalBox_49
   widget=Text_Body
   properties {
    autoWrapText=true
    bAutoCollapseWithEmptyText=true
    bIsScrollingEnabled=false
    lineHeightPercentage=1.5
    mobileFontSizeMultiplier=1.2
    text="[PH] A multi-line body text for each entry. Usually 2-3 lines. We explain something here."
    colorAndOpacity {colorUseRule=UseColor_Specified, specifiedColor=<rgba 0.5776 0.5776 0.5776 1>}
    font {
     bForceMonospaced=false
     bMaterialIsStencil=false
     fontMaterial=null
     fontObject=&HeadingNow
     letterSpacing=0
     monospacedWidth=1
     size=15.75
     skewAmount=0
     typefaceFontName=74 Regular
     outlineSettings {
      bApplyOutlineToDropShadows=false
      bMiteredCorners=false
      bSeparateFillAlpha=false
      outlineMaterial=null
      outlineSize=0
      outlineColor=<rgba 0 0 0 1>
     }
    }
    shadowOffset=<0 0>
   }
   slot {class=VerticalBoxSlot}
  }
 ]
}
