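# WBP_AboutItem_Dialog is an activatable UIKit dialog that shows "About Item" details for a Fortnite item.
# On activation it async-loads the view model class from the item's locker category via gameplay tag, deactivating if none is defined or the cast fails.
# The view model's entries fill a dynamic box of title, body and image widgets, reparented into the dialog's text scroll box.
# PreConstruct builds a design-time preview; button clicks trigger no action.
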WidgetBlueprint /FortniteCoreUI/Frontend/ItemPreview/AboutItemDialog/WBP_AboutItem_Dialog
graphs [
 ApplyDialogVM {nodes [
   n1 = ApplyDialogVM!() -> (then, DialogVM:UI Kit Dialog Viewmodel Object Reference) {is=title K2Node_FunctionEntry}
   n2 = Get Dialog_Internal() -> (Dialog_Internal:WBP UIKit Dialog Internal Object Reference)
   > DetermineWidgetsUsedBasedOnVM!(execute, self:WBP UIKit Dialog Internal Object Reference=@n2, DialogVM:UI Kit Dialog Viewmodel Object Reference=@n1) -> (then)
   > SetDialogData!(execute, self:WBP UIKit Dialog Internal Object Reference=@n2, Data:UI Kit Dialog Viewmodel Object Reference=@n1) -> (then)
  ]}
 EventGraph {nodes [
   n1 = OnButtonClicked!() -> (OutputDelegate, then, ButtonIndex:int) {is=title K2Node_CustomEvent}
   > Switch on Int!(execute:Exec, Selection:int=@n1.ButtonIndex) -> (Default:Exec)
   n2 = Event PreConstruct()
   > Branch(Condition=@n2.IsDesignTime)
   > Set DesignTimePreview_DialogVM!(
    execute
    DesignTimePreview_DialogVM:UI Kit Dialog Viewmodel Class Reference=Get AboutItemVMClass() -> (AboutItemVMClass:Dialog VM About Item Base Class Reference)
    self:WBP UIKit Dialog Internal Object Reference=Get Dialog_Internal()
   ) -> (then, Output_Get:UI Kit Dialog Viewmodel Class Reference)
   > n3 = Construct Object from Class(Class=Get AboutItemVMClass()) -> (ReturnValue:Dialog VM About Item Base Object Reference)
   > ApplyDialogVM!(execute, self, DialogVM:UI Kit Dialog Viewmodel Object Reference=@n3) -> (then) {is=title K2Node_CallFunction}
   > DesignTimePreset(self:WBP UIKit Dialog Internal Object Reference=Get Dialog_Internal())
   > n4 = GetDesignTimeEntries!(execute, self:Dialog VM About Item Base Object Reference=@n3) -> (then, Entries:Array of S About Item Entry Structures)
   > OnEntriesSet!(execute, self, Entries:Array of S About Item Entry Structures=@n4) -> (then)
   Event On Activated()
   > Bind Event to Dialog Button Clicked Dispatcher(self:WBP UIKit Dialog Internal Object Reference=Get Dialog_Internal(), Delegate=@n1) {is=title K2Node_AddDelegate}
   n5 = GetItemCategory!(self:Fort Item VM Object Reference=Get FortItemVM() -> (FortItemVM:Fort Item VM Object Reference)) -> (ReturnValue:Fort Locker Category Config Object Reference)
   n6 = TryGetSoftCustomDialogVMClass!(self:Fort Locker Category Config Object Reference=@n5, ViewRequest:Gameplay Tag Structure=(TagName="UI.ItemShop.AboutItemDialogClass")) -> (DialogVMClass:Object Soft Class Reference, ReturnValue:bool)
   > n7 = Branch(Condition=@n6.ReturnValue)
   > n8 = Async Load Class Asset!(execute:Exec, AssetClass:Object Soft Class Reference=@n6) -> (then:Exec, Completed:Exec, Class:Object Class Reference) {is=title K2Node_LoadAssetClass}
   n9 = Construct Object from Class(execute=@n8.Completed, Class=@n8) -> (ReturnValue:Object Reference)
   > n10 = Cast To DialogVM_AboutItem_Base!(execute, Object=@n9) -> (then, CastFailed, AsDialog VM About Item Base:Dialog VM About Item Base Object Reference) {is=title K2Node_DynamicCast}
   > n11 = Set DialogVM_ItemInfo(DialogVM_ItemInfo:Dialog VM About Item Base Object Reference=@n10) -> (Output_Get:Dialog VM About Item Base Object Reference)
   > n12 = GetEntriesForItem!(execute, self:Dialog VM About Item Base Object Reference=@n11, ItemVM:Fort Item VM Object Reference=Get FortItemVM()) -> (then, Entries:Array of S About Item Entry Structures)
   > OnEntriesSet(Entries=@n12)
   > ForceLayoutPrepass(self:Widget Object Reference)
   n13 = PrintString!(
    execute=@n10.CastFailed
    InString="To String (Text)"(InText=Format Text!(
      Format="INVTEXT(\"[WBP_AboutItem_Dialog] Unable to cast category VM for {ItemCategory} with ID \\\"UI.ItemShop.AboutItemDialogClass\\\" to DialogVM_AboutItemBase\")"
      ItemCategory:Fort Locker Category Config Object Reference=@n5
     ) -> (Result))
    bPrintToScreen
    bPrintToLog
    TextColor
    Duration
    Key
   ) -> (then)
   DeactivateWidget(execute=[@n13.then, @n14.then])
   n14 = PrintString(
    execute=@n7.else
    InString="To String (Text)"(InText=Format Text(
      Format="INVTEXT(\"[WBP_AboutItem_Dialog] No Custom Dialoge VMClass defined for {ItemCategory} with ID \\\"UI.ItemShop.AboutItemDialogClass\\\"\")"
      ItemCategory=@n5
     ))
    TextColor=(R=1,G=0,B=0,A=1)
   )
   Event On Deactivated()
   > Unbind Event from Dialog Button Clicked Dispatcher(self:WBP UIKit Dialog Internal Object Reference=Get Dialog_Internal(), Delegate=@n1) {is=title K2Node_RemoveDelegate}
  ]}
 OnEntriesSet {nodes [
   n1 = OnEntriesSet!() -> (then, Entries:Array of S About Item Entry Structures) {is=title K2Node_FunctionEntry}
   > Reset!(
    execute
    self:Dynamic Entry Box Object Reference=Get DynamicEntryBox_Entries() -> (DynamicEntryBox_Entries:Fort Dynamic Entry Box Object Reference)
    bDeleteWidgets:bool=false
   ) -> (then)
   > n2 = Branch(Condition=Is Empty!(TargetArray:Array of S About Item Entry Structures=@n1) -> (ReturnValue:bool))
   AddChild(
    execute=@n2.else
    self=Get ScrollBox_Text!(self:WBP UIKit Dialog Internal Object Reference=Get Dialog_Internal()) -> (ScrollBox_Text:Scroll Box Object Reference)
    Content=Get Overlay_Entries() -> (Overlay_Entries:Overlay Object Reference)
   )
   > n3 = For Each Loop(Array:Array of S About Item Entry Structures=@n1) -> (Array Element:S About Item Entry Structure)
   > n4 = Create Entry!(execute, self:Dynamic Entry Box Object Reference=Get DynamicEntryBox_Entries()) -> (then, ReturnValue:User Widget Object Reference)
   > n5 = Cast To WBP_AboutItemDialog_BodyEntry!(execute, Object=@n4) -> (then, CastFailed, AsWBP About Item Dialog Body Entry:WBP About Item Dialog Body Entry Object Reference) {is=title K2Node_DynamicCast}
   n6 = Break S About Item Entry!(S_AboutItemEntry:"S About Item Entry Structure (by ref)"=@n3) -> (
    Title_6_#1:Text
    BodyText_7_#2:Text
    Image_8_#3:Texture 2D Soft Object Reference
   ) {is=title K2Node_BreakStruct}
   > InitializeEntry!(
    execute
    self:WBP About Item Dialog Body Entry Object Reference=@n5
    Title:Text=@n6
    BodyText:Text=@n6.BodyText_7_#2
    Image:Texture 2D Soft Object Reference=@n6.Image_8_#3
   ) -> (then)
  ]}
]
variables [AboutItemVMClass {type=Class, default=&/DialogVM_AboutItem_Base.DialogVM_AboutItem_Base_C}]
widgets {
 info {
  inheritedWidgetCount=0
  parentClass=&CommonActivatableWidget
  rootWidgetClass=&CommonVisibilitySwitcher
  widgetCount=4
 }
 widgets [
  CommonVisibilitySwitcher_1 {widgetClassPath=&CommonVisibilitySwitcher, widget=CommonVisibilitySwitcher_1, properties {bOverride_Cursor=true}}
  Dialog_Internal {
   widgetClassPath=/FortUILibrary/Systems/Dialog/WBP_UIKit_Dialog_Internal
   bIsVariable=true
   parent=CommonVisibilitySwitcher_1
   widget=Dialog_Internal
   properties {
    designTimePreview_DialogVM=&DialogVM_AboutItem_Base_C
    backgroundProperties {
     material=/FortUILibrary/UIKit/Dialogs/MaterialInstances/MI_UI_DialogBackground
     image Size=<32 32>
     placement {
      horizontal Alignment=HAlign_Fill
      vertical Alignment=VAlign_Fill
      grid {column=0, column Span=2, layer=1, row=0, row Span=0, nudge=<0 0>}
      padding=<ltrb 0 0 0 0>
     }
    }
    buttons {
     each Button Horizontal Fill=false
     each Button Min Width=0
     each Button Size=Medium
     focuse Button=0
     space Between Buttons=30
     cTA Buttons Types [Regular,Quiet]
     group Placement {
      horizontal Alignment=HAlign_Fill
      vertical Alignment=VAlign_Bottom
      grid {column=1, column Span=0, layer=0, row=2, row Span=0, nudge=<0 0>}
      padding=<ltrb 80 64 80 48>
     }
    }
    dialogSIze {
     height Override {enable=false, value=730}
     max Aspect Ratio {enable=false, value=0}
     max Desired Height {enable=true, value=800}
     max Desired Width {enable=false, value=0}
     min Aspect Ratio {enable=false, value=0}
     min Desired Height {enable=false, value=0}
     min Desired Width {enable=false, value=0}
     width Override {enable=true, value=1424}
    }
    imageProperties {
     material=/FortUILibrary/Systems/Dialog/MaterialInstances/MI_UI_DialogImage
     image Size=<512 730>
     placement {
      horizontal Alignment=HAlign_Fill
      vertical Alignment=VAlign_Fill
      grid {column=1, column Span=0, layer=20, row=0, row Span=1, nudge=<0 0>}
      padding=<ltrb 0 0 0 0>
     }
    }
    textProperties {
     auto Wrap Text=true
     justification=Left
     line Height Percentage=1.3
     mobile Font Size Multiplier=1.3
     transform Policy=None
     wrap Text At=0
     color {colorUseRule=UseColor_Specified, specifiedColor=<rgba 0.5776 0.5776 0.5776 1>}
     font {
      bForceMonospaced=false
      bMaterialIsStencil=false
      fontMaterial=null
      fontObject=&HeadingNow
      letterSpacing=10
      monospacedWidth=1
      size=15.75
      skewAmount=0
      typefaceFontName=74 Regular
      outlineSettings {
       bApplyOutlineToDropShadows=false
       bMiteredCorners=false
       bSeparateFillAlpha=false
       outlineMaterial=null
       outlineSize=0
       outlineColor=<rgba 0 0 0 1>
      }
     }
     placement {
      horizontal Alignment=HAlign_Fill
      vertical Alignment=VAlign_Top
      grid {column=1, column Span=0, layer=0, row=1, row Span=0, nudge=<0 0>}
      padding=<ltrb 80 0 80 0>
     }
    }
    titleProperties {
     auto Wrap Text=true
     justification=Left
     line Height Percentage=1
     mobile Font Size Multiplier=1.3
     transform Policy=ToUpper
     wrap Text At=0
     color {colorUseRule=UseColor_Specified, specifiedColor=<rgba 1 1 1 1>}
     font {
      bForceMonospaced=false
      bMaterialIsStencil=false
      fontMaterial=null
      fontObject=&HeadingNow
      letterSpacing=-10
      monospacedWidth=1
      size=25.5
      skewAmount=0
      typefaceFontName=86 Bold Italic
      outlineSettings {
       bApplyOutlineToDropShadows=false
       bMiteredCorners=false
       bSeparateFillAlpha=false
       outlineMaterial=null
       outlineSize=0
       outlineColor=<rgba 0 0 0 1>
      }
     }
     placement {
      horizontal Alignment=HAlign_Fill
      vertical Alignment=VAlign_Top
      grid {column=1, column Span=0, layer=0, row=0, row Span=0, nudge=<0 0>}
      padding=<ltrb 80 80 80 24>
     }
    }
   }
   slot {class=CommonVisibilitySwitcherSlot}
  }
  Overlay_Entries {
   widgetClassPath=&Overlay
   bIsVariable=true
   parent=CommonVisibilitySwitcher_1
   widget=Overlay_Entries
   properties {bOverride_Cursor=true}
   slot {class=CommonVisibilitySwitcherSlot}
  }
  DynamicEntryBox_Entries {
   widgetClassPath=/Script/FortniteUI.FortDynamicEntryBox
   bIsVariable=true
   parent=Overlay_Entries
   widget=DynamicEntryBox_Entries
   properties {
    entryBoxType=Vertical
    entryWidgetClass=&/WBP_AboutItemDialog_BodyEntry.WBP_AboutItemDialog_BodyEntry_C
    entrySpacing=<0 48>
   }
   slot {class=OverlaySlot, properties {padding=<ltrb 0 24 24 0>}}
  }
 ]
}
